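# Translucent, unlit, two-sided Niagara mesh and ribbon material for the rocket engine flames on the Realm Male VinderTech glider.
# Panning samples of one distortion texture plus a UV gradient "fire line", whose contrast follows a dynamic parameter, are tinted by particle color for HDR emissive.
# Opacity combines hardened flame noise, particle alpha, UV falloffs and counter-scrolling UV1 noise, with depth fade off mobile.
# Red vertex color masks a local-axis offset driven by the WPO Stretch parameter.
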Material /Game/Gadgets/Assets/VinderTech_GliderChute/Glider_Realm_Male/FX/M_Realm_Rocket_Engines
graph {
 properties {
  bUsedWithNiagaraMeshParticles=true
  bUsedWithNiagaraRibbons=true
  blendMode=BLEND_Translucent
  pixelDepthOffsetMode=PDOM_Legacy
  shadingModel=MSM_Unlit
  twoSided=true
  usedShadingModels=Unlit
 }
 nodes [
  n1 = ParticleColor()
  n2 = OneMinus(None=ComponentMask(None=TextureCoordinate()) {r=true})
  n3 = LinearInterpolate(
   A=ScalarParameter() {defaultValue=2, parameterName=FireLineContrast A}
   B=ScalarParameter() {defaultValue=5, parameterName=FireLineContrast B}
   Alpha=DynamicParameter() {defaultValue=<rgba 1 1 1 20>, paramNames [FireLineControl, WPO Stretch, Param3, Param4]}.FireLineControl
  ) {constA=2, constB=5}
  n4 = ComponentMask(None=VectorParameter() {group=UVScale, parameterName=UVScale, defaultValue=<rgba 0.5 1 0 0>}.RGB) {g=true, r=true}
  n5 = MaterialFunctionCall(XY=Multiply(A=TextureCoordinate() {uTiling=0.5}, B=@n4)) {materialFunction=&Swizzle}
  n6 = MaterialFunctionCall(
   In=TextureSample(UVs=Panner(Coordinate=@n5.YX) {speedX=-0.05, speedY=-1}) {texture=/Game/Global/Textures/Default/Masks_and_Gradients/T_BlasterBeamDistort}.G
   Contrast=Constant() {r=0.2}
  ) {materialFunction=&CheapContrast}
  n7 = Multiply(A=@n4, B=TextureCoordinate() {vTiling=4})
  n8 = Multiply(A=Time(), B=ScalarParameter() {defaultValue=5, parameterName=Speed})
  n9 = TextureSample(UVs=Panner(Coordinate=MaterialFunctionCall(XY=@n7) {materialFunction=&Swizzle}.YX, Time=@n8) {speedX=0.1, speedY=-1}) {texture=&T_BlasterBeamDistort}
  n10 = TextureCoordinate() {uTiling=0.6, vTiling=2.5}
  n11 = Panner(
   Coordinate=MaterialFunctionCall(XY=Multiply(A=@n4, B=@n10)) {materialFunction=&Swizzle}.YX
   Time=@n8
  ) {speedX=-0.3, speedY=-1}
  n12 = Add(
   A=Multiply(A=Power(Base=@n2, Exp=@n3), B=ScalarParameter() {defaultValue=0.6, parameterName=FireLineIntensity})
   B=Multiply(A=Add(A=@n6.Result, B=@n9.G), B=TextureSample(UVs=@n11) {texture=&T_BlasterBeamDistort}.G)
  )
  n13 = Multiply(A=@n12, B=@n1.RGB)
  MaterialFunctionCall(Emissive=MaterialFunctionCall(LDR Input=Clamp(None=@n13), HDR Tint=@n13) {materialFunction=&DeriveHDRfromLDR}.RemappedRange) {materialFunction=&MF_FX_EmissiveAdjustments}
  n14 = Constant3Vector() {constant=<rgba 0 1 0 0>}
  Multiply(
   A=VertexColor().R
   B=Multiply(
    A=DynamicParameter() {defaultValue=<rgba 1 1 1 20>, paramNames [FireLineControl, WPO Stretch, Param3, Param4]}.WPO Stretch
    B=Transform(None=@n14) {transformSourceType=TRANSFORMSOURCE_Local}
   )
  )
  n15 = Clamp(None=MaterialFunctionCall(In=Power(Base=@n12) {constExponent=0.5}, Contrast=ScalarParameter() {defaultValue=0.5, parameterName=HardEdgeness}) {materialFunction=&CheapContrast}.Result)
  n16 = ComponentMask(None=TextureCoordinate()) {r=true}
  n17 = ComponentMask(None=TextureCoordinate() {coordinateIndex=1}) {r=true}
  n18 = Multiply(A=Power(Base=OneMinus(None=@n17))) {constB=10}
  n19 = Multiply(
   A=TextureCoordinate() {coordinateIndex=1}
   B=ComponentMask(None=VectorParameter() {parameterName=UV1 Scale, defaultValue=<rgba 0 5 0 0>}.RGB) {g=true, r=true}
  )
  n20 = TextureSample(UVs=Panner(Coordinate=MaterialFunctionCall(XY=@n19) {materialFunction=&Swizzle}.YX) {speedX=-1}) {texture=&T_BlasterBeamDistort}
  n21 = Multiply(
   A=TextureCoordinate() {coordinateIndex=1}
   B=ComponentMask(None=VectorParameter() {parameterName=UV1 Scale2, defaultValue=<rgba 0 2 0 0>}.RGB) {g=true, r=true}
  )
  n22 = TextureSample(UVs=Panner(Coordinate=MaterialFunctionCall(XY=@n21) {materialFunction=&Swizzle}.YX) {speedX=1}) {texture=&T_BlasterBeamDistort}
  n23 = Multiply(A=Multiply(A=Multiply(A=@n1.A, B=@n15), B=OneMinus(None=@n16)), B=Multiply(A=Saturate(None=@n18), B=Add(A=@n20.G, B=@n22.G)))
  ShadingPathSwitch(
   Default=DepthFade(Opacity=@n23, FadeDistance=ScalarParameter() {defaultValue=10, group=Opacity, parameterName=Depth Fade Distance, sortPriority=1})
   Mobile=@n23
  ) {desc=Migrated from FeatureLevelSwitch}
 ]
}
