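# M_SimpleParticle_HDR is a translucent, two-sided material for Niagara sprite particles in the Realm Male glider's FX.
# It uses no textures, taking its color from particle color RGB.
# DeriveHDRfromLDR receives that color clamped as its LDR input and unclamped as its HDR tint.
# The remapped range feeds the emissive input of MF_FX_EmissiveAdjustments.
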Material /Game/Gadgets/Assets/VinderTech_GliderChute/Glider_Realm_Male/FX/M_SimpleParticle_HDR
graph {
 properties {bUsedWithNiagaraSprites=true, blendMode=BLEND_Translucent, pixelDepthOffsetMode=PDOM_Legacy, twoSided=true, usedShadingModels=Default Lit}
 nodes [
  n1 = ParticleColor()
  MaterialFunctionCall(Emissive=MaterialFunctionCall(LDR Input=Clamp(None=@n1.RGB), HDR Tint=@n1.RGB) {materialFunction=&DeriveHDRfromLDR}.RemappedRange) {materialFunction=&MF_FX_EmissiveAdjustments}
 ]
}
